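; MKART_LobbyBG_Interface is a Blueprint interface for lobby background actors in the frontend UI maps.
; It declares one function, MKARTLobbyBG_Interface, which passes a Frontend Background Data Structure as NewBG.
; The file holds only the signature, so implementers decide how the new background data is applied.
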(fn MKARTLobbyBG_Interface
  (params (NewBG :type Frontend%20Background%20Data%20Structure)))
